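# WorldAlignedBlend is a library material function that builds a blend alpha from how closely surface normals face a world vector, default +Z.
# Three outputs use pixel normals, vertex normals or a transformed explicit normal, the latter two allowing normal-map blending without reentrancy.
# Each remaps the dot product to 0–1, applies sharpness and bias, then optionally clamps.
# The Alpha input reaches only the explicit-normal output; several leftover nodes are unconnected.
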MaterialFunction /Engine/Functions/Engine_MaterialFunctions01/AlphaBlend/WorldAlignedBlend
graph {
 properties {bExposeToLibrary=true}
 nodes [
  n1 = FunctionInput(Preview=Constant() {r=-1}) {
   bUsePreviewValueAsDefault=true
   description=Shift the alpha either up or down.
   inputName=Blend Bias
   inputType=FunctionInput_Scalar
   sortPriority=3
  }
  n2 = FunctionInput(Preview=Constant() {r=2}) {
   bUsePreviewValueAsDefault=true
   description=Blend Sharpness. Increasing this will cause the alpha to be brighter facing the vector.
   inputName=Blend Sharpness
   inputType=FunctionInput_Scalar
   sortPriority=2
  }
  n3 = Multiply(A=@n2) {constB=0.5}
  Divide(A=@n1, B=@n3)
  n4 = FunctionInput(Preview=Constant3Vector() {constant=<rgba 0 0 1 0>}) {
   bUsePreviewValueAsDefault=true
   description=Explicit Normal. Useful for when blending normal maps and you still want a noisy blend
   inputName=In Explicit Normal
  }
  n5 = Constant3Vector() {constant=<rgba 0 0 1 0>}
  n6 = Normalize(VectorInput=FunctionInput(Preview=@n5) {bUsePreviewValueAsDefault=true, description="Vector for Blend. Default = 0,0,1", inputName=In World Vector, sortPriority=1})
  n7 = DotProduct(A=Normalize(VectorInput=Transform(None=@n4)), B=@n6)
  n8 = Constant() {r=0.5}
  n9 = FunctionInput(Preview=Constant() {r=1}) {
   bUsePreviewValueAsDefault=true
   description="Lets you specify an alpha to apply on top of the world. Good use would be a Vertex Color.Default= 1 which means no alpha."
   inputName=Alpha
   inputType=FunctionInput_Scalar
   sortPriority=8
  }
  n10 = Multiply(A=Add(A=Multiply(A=@n7, B=@n8), B=@n8), B=@n9)
  n11 = Subtract(A=@n1, B=@n3)
  n12 = Add(A=Multiply(A=@n10, B=@n2), B=@n11)
  n13 = FunctionInput(Preview=StaticBool() {value=true}) {
   bUsePreviewValueAsDefault=true
   description=Only needed when you want to use this data downstream to modulate another alpha and negatives values are desired.
   inputName=Clamped?
   inputType=FunctionInput_StaticBool
   sortPriority=7
  }
  FunctionOutput(None=StaticSwitch(True=Clamp(None=@n12), False=@n12, Value=@n13)) {
   description=Useful if you want a specific normal map to affect the blend. Useful if you need the added detail and need to blend a normal map. For a cliff/grass material use this and specify the Rock's normal on the input side.
   outputName=w/ Explicit Normal
   sortPriority=5
  }
  n14 = PixelNormalWS()
  n15 = Constant() {r=0.5}
  n16 = Add(A=Multiply(A=DotProduct(A=@n14, B=@n6), B=@n15), B=@n15)
  n17 = Add(A=Multiply(A=@n16, B=@n2), B=@n11)
  FunctionOutput(None=StaticSwitch(True=Clamp(None=@n17), False=@n17, Value=@n13)) {
   description="Default Output uses PixelNormalWS. Gives the best result but won't allow you to blend normal maps, since it would be a reentrant expression. Use one of the other two outputs for blending Normals."
   outputName=Alpha
   sortPriority=1
  }
  n18 = VertexNormalWS()
  n19 = Constant() {r=0.5}
  n20 = Add(A=Multiply(A=DotProduct(A=@n18, B=@n6), B=@n19), B=@n19)
  n21 = Add(A=Multiply(A=@n20, B=@n2), B=@n11)
  FunctionOutput(None=StaticSwitch(True=Clamp(None=@n21), False=@n21, Value=@n13)) {
   description="The simplest blend that will only be based on the mesh's vertices, not any normal maps. Will work anywhere."
   outputName=w/Vertex Normals
   sortPriority=5
  }
  LinearInterpolate(A=@n2, B=@n1, Alpha=Multiply(A=@n20, B=@n9))
  LinearInterpolate(A=@n2, B=@n1, Alpha=@n10)
  LinearInterpolate(A=@n2, B=@n1, Alpha=Multiply(A=@n16, B=@n9))
  Normalize(VectorInput=Transform(None=@n5))
 ]
}
